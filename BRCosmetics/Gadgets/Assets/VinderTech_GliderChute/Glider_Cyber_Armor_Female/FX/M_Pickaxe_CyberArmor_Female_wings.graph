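# Unlit, translucent, two-sided material for glowing circuit wings on the female Cyber Armor cosmetic, named for a pickaxe but stored with the glider.
# A rounded local-space X mask offsets UVs per wing side, while the circuit texture's green channel offsets panning time for flame noise.
# Circuit-masked noise plus a time-of-day-adjusted gradient with static feeds HDR remapping and emissive adjustment.
# Alternate color, vertex-color and sine-flicker opacity branches remain, one labeled experiments.
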Material /BRCosmetics/Gadgets/Assets/VinderTech_GliderChute/Glider_Cyber_Armor_Female/FX/M_Pickaxe_CyberArmor_Female_wings
graph {
 properties {
  bUsedWithMeshDeformer=true
  bUsedWithSkeletalMesh=true
  blendMode=BLEND_Translucent
  pixelDepthOffsetMode=PDOM_Legacy
  shadingModel=MSM_Unlit
  twoSided=true
  usedShadingModels=Unlit
 }
 nodes [
  n1 = ScalarParameter() {defaultValue=0.05, parameterName=Gradient low}
  n2 = ScalarParameter() {defaultValue=1, parameterName=Gradient High}
  n3 = WorldPosition() {worldPositionShaderOffset=WPT_ExcludeAllShaderOffsets}
  n4 = PreSkinnedLocalBounds()
  n5 = Divide(
   A=Subtract(A=TransformPosition(None=@n3.XYZ) {transformSourceType=TRANSFORMPOSSOURCE_World, transformType=TRANSFORMPOSSOURCE_Instance}, B=@n4.Min)
   B=@n4.Extents
  )
  n6 = Multiply(A=Round(None=ComponentMask(None=@n5) {r=true})) {constB=0.5}
  n7 = Add(A=TextureCoordinate() {uTiling=2, vTiling=2}, B=@n6)
  n8 = TextureSample() {
   samplerType=SAMPLERTYPE_LinearColor
   texture=&/T_Cyber_armor_wings_Circut_basecolor
  }
  n9 = Add(A=Multiply(A=ScalarParameter() {defaultValue=0.3, parameterName=circut offset}, B=Subtract(A=@n8.G) {constA=0.5, constB=0.5}), B=Time())
  n10 = TextureSample(UVs=Panner(
    Coordinate=MaterialFunctionCall(XY=@n7) {materialFunction=&Swizzle}.YX
    Time=@n9
    Speed=Constant2Vector() {g=0.5}
   ) {speedY=-0.3}) {texture=/Game/Global/Textures/Default/Masks_and_Gradients/gradient_stasis}
  n11 = ScalarParameter() {defaultValue=0.01, parameterName=Noise low}
  n12 = ScalarParameter() {defaultValue=1, parameterName=Noise High}
  n13 = TextureCoordinate()
  n14 = TextureSample(UVs=Panner(Coordinate=Add(A=@n13, B=@n6), Time=@n9, Speed=Constant2Vector() {r=0.5}) {speedY=-0.3}) {samplerType=SAMPLERTYPE_LinearGrayscale, texture=/BRCosmetics/Weapons/FORT_Melee/Pickaxe_Cube_Ninja_Male/FX/T_FluidFlameTile}
  n15 = Multiply(
   A=Multiply(A=LinearInterpolate(A=@n1, B=@n2, Alpha=@n10.R) {constA=0.1}, B=LinearInterpolate(A=@n11, B=@n12, Alpha=@n14.R) {constA=0.1})
   B=Multiply(A=@n8.R, B=ScalarParameter() {defaultValue=3, parameterName=circut inten}) {constB=3}
  )
  n16 = MaterialFunctionCall() {materialFunction=&LinearGradient}
  n17 = Multiply(
   A=Power(Base=OneMinus(None=@n16.UGradient), Exp=ScalarParameter() {defaultValue=3, parameterName=gradient exp}) {constExponent=3}
   B=ScalarParameter() {defaultValue=1, parameterName=gradient mul}
  )
  n18 = ScalarParameter() {parameterName=gradient add}
  n19 = Panner(Coordinate=TextureCoordinate(), Speed=Constant2Vector() {r=0.5}) {speedY=-0.3}
  n20 = Multiply(
   A=MaterialFunctionCall(In=Add(A=@n17, B=@n18), Contrast=ScalarParameter() {parameterName=gradient contrast}) {materialFunction=&CheapContrast}.Result
   B=Add(A=TextureSample(UVs=@n19) {texture=/Game/Athena/Items/Gameplay/Second_Chance/Textures/T_Van_TV_Static_01_D}.RGB, B=Constant() {r=0.9})
  )
  Add(A=@n15, B=@n20)
  n21 = VectorParameter() {parameterName=Wing Color inner, defaultValue=<rgba 0.5 0.017 0.0845 1>}
  n22 = VectorParameter() {parameterName=Wing Color outer, defaultValue=<rgba 2 0.0679 0.3381 1>}
  n23 = NamedRerouteDeclaration(None=LinearInterpolate(
    A=LinearInterpolate(A=@n21.RGB, B=@n22.RGB, Alpha=@n20)
    B=VectorParameter() {parameterName=Wing Color Circut, defaultValue=<rgba 2 0.5 0.7098 1>}.RGB
    Alpha=@n15
   )) {name=winged color Param, nodeColor=<rgba 0.7109 1 0 1>}
  n24 = VectorParameter() {
   parameterName=Dynamic Parameter 0
   channelNames {a=Na, b=VertexColorInside, g=VertexColorOutside, r=IntensityModifier}
   defaultValue=<rgba 1 1 0 0>
  }
  n25 = VertexColor()
  n26 = Multiply(A=@n24.VertexColorInside, B=Divide(A=@n25.B, B=@n25.B))
  n27 = Add(
   A=MaterialFunctionCall(
    LDR Input=@n23
    HDR Tint=Saturate(None=@n23)
    HDR Intensity Multiplier=@n24.IntensityModifier
    Derive Luminance from Linear=StaticBool() {value=true}
   ) {materialFunction=&DeriveHDRfromLDR}.RemappedRange
   B=Multiply(A=Clamp(None=@n26) {maxDefault=30}, B=NamedRerouteUsage())
  )
  n28 = MaterialFunctionCall!(X=@n25.G, Tangent 0=Constant(), Tangent 1=Constant() {r=2.2}) {materialFunction=/Engine/Functions/Engine_MaterialFunctions02/Gradients/SmoothCurve}
  Add(A=@n27, B=Multiply(A=Multiply(A=@n24.VertexColorOutside, B=@n28.Result), B=NamedRerouteUsage()))
  Constant()
  Constant() {r=0.25}
  Constant() {r=0.5}
  DynamicParameter() {defaultValue=<rgba 0.25 1 1 1>, paramNames [IntensityModifier, VertexColorOutside, VertexColorInside, Param4]}
  DynamicParameter() {defaultValue=<rgba 0.25 1 1 1>, paramNames [IntensityModifier, VertexColorOutside, VertexColorInside, Param4]}
  DynamicParameter() {defaultValue=<rgba 0.25 1 1 1>, paramNames [IntensityModifier, VertexColorOutside, VertexColorInside, Param4]}
  Frac(None=@n9)
  n29 = VectorParameter() {parameterName=Opac sine time Offsets, defaultValue=<rgba 0.232 0.88 1.0324 0.416>}
  n30 = Time()
  n31 = VectorParameter() {parameterName=Opac time speed Multipliers, defaultValue=<rgba 5 0.74 1.1 1.0829>}
  n32 = AppendVector(A=@n31.RGB, B=@n31.RGB)
  n33 = MaterialFunctionCall(Float4=Add(A=AppendVector(A=@n29.RGB, B=@n29.RGB), B=Multiply(A=@n30, B=@n32))) {materialFunction=&BreakOutFloat4Components}
  n34 = ConstantBiasScale(None=Sine(None=@n33.R))
  n35 = ConstantBiasScale(None=Sine(None=@n33.G))
  n36 = ConstantBiasScale(None=Sine(None=@n33.B))
  n37 = ConstantBiasScale(None=Sine(None=@n33.A))
  n38 = VectorParameter() {parameterName="Opac sine  Multipliers", defaultValue=<rgba 0.936 1.4756 0.62 0.3421>}
  n39 = MaterialFunctionCall(Float4=Multiply(
    A=MaterialFunctionCall(X=@n34, Y=@n35, Z=@n36, A=@n37) {materialFunction=&MakeFloat4}.Result
    B=AppendVector(A=@n38.RGB, B=@n38.RGB)
   )) {materialFunction=&BreakOutFloat4Components}
  n40 = Multiply(A=@n39.R, B=@n39.R)
  n41 = Multiply(A=@n39.B, B=@n39.B)
  LinearInterpolate(
   A=ScalarParameter() {defaultValue=1, parameterName=Opacity low}
   B=ScalarParameter() {defaultValue=1, parameterName=Opacity High}
   Alpha=Saturate(None=Multiply(A=@n40, B=@n41))
  )
  MaterialFunctionCall(Contrast=Constant() {r=1}) {materialFunction=&CheapContrast}
  n42 = PreSkinnedLocalBounds()
  n43 = Subtract(
   A=TransformPosition(None=WorldPosition() {worldPositionShaderOffset=WPT_ExcludeAllShaderOffsets}.XYZ) {transformSourceType=TRANSFORMPOSSOURCE_World, transformType=TRANSFORMPOSSOURCE_Instance}
   B=@n42.Min
  )
  n44 = Round(None=ComponentMask(None=Divide(A=@n43, B=@n42.Extents)) {r=true})
  n45 = Add(
   A=TextureCoordinate()
   B=Multiply(
    A=NamedRerouteDeclaration(None=@n44) {name=Left Right Wing Mask, nodeColor=<rgba 0.3594 1 0 1>}
    B=ScalarParameter() {defaultValue=0.5, parameterName=Left Right Offset}
   ) {constB=0.5}
  )
  n46 = TextureSample() {
   samplerType=SAMPLERTYPE_LinearColor
   texture=&T_Cyber_armor_wings_Circut_basecolor
  }
  n47 = Add(A=Multiply(A=ScalarParameter() {defaultValue=0.3, parameterName=circut offset}, B=Subtract(A=@n46.G) {constA=0.5, constB=0.5}), B=Time())
  n48 = Constant2Vector() {r=0.5}
  n49 = TextureSample() {
   samplerType=SAMPLERTYPE_LinearColor
   texture=&T_Cyber_armor_wings_Circut_basecolor
  }
  n50 = Multiply(
   A=TextureSample(UVs=Panner(Coordinate=@n45, Time=@n47, Speed=@n48) {speedY=-0.3}) {samplerType=SAMPLERTYPE_LinearGrayscale, texture=&T_FluidFlameTile}.R
   B=@n49.R
  )
  n51 = MaterialFunctionCall() {materialFunction=&LinearGradient}
  n52 = LinearInterpolate(A=OneMinus(None=@n51.UGradient), B=@n51.UGradient, Alpha=ScalarParameter() {parameterName=invert Gradient})
  n53 = VectorParameter() {parameterName=Gradient Options, channelNames {a=contrast, b=add, g=Mul, r=Pow}, defaultValue=<rgba 1 1 0 0>}
  n54 = VectorParameter() {parameterName=Gradient Options night, channelNames {a=contrast, b=add, g=Mul, r=Pow}, defaultValue=<rgba 1 1 0 0>}
  n55 = MaterialFunctionCall(Float3=MaterialFunctionCall(Color_Morning=@n53, Color_Day=@n53, Color_Evening=@n53, Color_Night=@n54) {materialFunction=&Color_Time_of_Day}.Result) {materialFunction=&BreakOutFloat3Components}
  n56 = Add(A=Multiply(A=Power(Base=@n52, Exp=@n55.R) {constExponent=3}, B=@n55.G), B=@n55.B)
  n57 = MaterialFunctionCall(Value_Morning=@n53.contrast, Value_Day=@n53.contrast, Value_Evening=@n53.contrast, Value_Night=@n54.contrast) {materialFunction=&Scalar_Time_of_Day}
  n58 = Panner(Coordinate=TextureCoordinate(), Speed=Constant2Vector() {r=0.5}) {speedY=-0.3}
  n59 = Multiply(
   A=Clamp(None=MaterialFunctionCall(In=@n56, Contrast=@n57.Result) {materialFunction=&CheapContrast}.Result)
   B=Add(A=TextureSample(UVs=@n58) {texture=&T_Van_TV_Static_01_D}.RGB, B=Constant() {r=0.9})
  )
  n60 = Clamp(None=Multiply(A=NamedRerouteUsage(), B=Add(A=@n50, B=@n59)))
  n61 = Saturate(None=NamedRerouteUsage())
  n62 = Multiply(A=NamedRerouteUsage(), B=ScalarParameter() {defaultValue=7.5, parameterName="HDR Intensity Multiplier (S)"})
  n63 = StaticBool() {value=true}
  n64 = NamedRerouteUsage()
  n65 = ScalarParameter() {defaultValue=0.8, parameterName=VertecxColorInisde}
  MaterialFunctionCall(Emissive=Add(
    A=MaterialFunctionCall(LDR Input=@n60, HDR Tint=@n61, HDR Intensity Multiplier=@n62, Derive Luminance from Linear=@n63) {materialFunction=&DeriveHDRfromLDR}.RemappedRange
    B=Multiply(A=@n64, B=@n65)
   )) {materialFunction=&MF_FX_EmissiveAdjustments}
  Multiply()
  n66 = VectorParameter() {parameterName=Color, defaultValue=<rgba 1 1 1 7.5>}
  NamedRerouteDeclaration(None=@n66.RGB) {name=Color RGB Param, nodeColor=<rgba 1 0 1 1>}
  NamedRerouteDeclaration(None=@n66.A) {name=Color A Param, nodeColor=<rgba 1 1 0 1>}
  ParticleColor()
  ParticleColor()
  Reroute(None=@n27) {desc=experiments}
  VertexColor()
  VertexColor()
 ]
}
