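# Niagara effect for the female Cyber Armor glider, adding energy glows, electric arcs and lights at four sampled bones or sockets of the user skeletal mesh.
# Mesh glows, flares and pulsing lights mark those points, while ribbon arcs and lightning sprites crackle around them.
# Airglow sprites stretch along each point's Z axis via SpriteBasedLine, with Timeline-driven alpha.
# All emitters share a system color, with alt-costume and time-of-day dimming on some.
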NiagaraSystem /BRCosmetics/Gadgets/Assets/VinderTech_GliderChute/Glider_Cyber_Armor_Female/FX/NS_CyberArmor_Female_Glider
dependencies {
 usedDataInterfaces [
  &NiagaraDataInterfaceArrayFloat
  &NiagaraDataInterfaceCurve
  &NiagaraDataInterfaceArrayFloat3
  &NiagaraDataInterfaceVectorCurve
 ]
 usedDynamicInputs []
 usedModules [
  &self:SystemSpawnScript.SystemScriptSource.SystemScriptGraph.NiagaraNodeAssignment_0.SetVariables_#1
  &self:SystemSpawnScript.SystemScriptSource.SystemScriptGraph.NiagaraNodeAssignment_1.SetVariables_#2
  &SystemState
  &Timeline
 ]
}
emitters {
 3d_glow {
  emitterUpdateScript [
   &EmitterState()
   &SpawnRate(SpawnRate=8, Use Spawn Probability=false)
  ]
  particleSpawnScript [
   &InitializeParticle(
    Color=&MultiplyLinearColor(Linear Color=$System.ColorPlusBoost, Scale Factor=&MakeVector4_FromFloat(Float=2.5))
    Lifetime Max=1.2
    Lifetime Min=0.8
    Lifetime Mode=::NewEnumerator1
    Mesh Scale=<0.3 0.2 0.2>
    Mesh Scale Mode=::NewEnumerator4
    Position Mode=::NewEnumerator0
    Write Material Random=true
   )
   &CameraOffset(Camera Offset Amount=5)
  ]
  particleUpdateScript [
   &ParticleState()
   &SkeletalMeshLocation(
    Bone / Socket Index=&ModuloInt(A=&ReturnParticleIDAttributes(), B=4)
    Bone Sampling Mode=::NewEnumerator4
    MaskSpawnGroup=false
    Skeletal Mesh=$User.SkeletalMesh
    Use Sampled Mesh Attribute as Ribbon ID=false
   )
   &DynamicMaterialParameters(
    Index 0 Param 1=&FloatFromCurve(FloatCurve=DataInterface() {propertyValues {Curve {keys [<at 0 1 RCIM_Cubic>,<at 0.8 0 RCIM_Cubic>]}}})
    Index 0 Param 2=2
    Index 0 Param 3=1
    Index 0 Param 4=1
    Index 1 Param 1=&FloatFromCurve(FloatCurve=DataInterface() {propertyValues {Curve {keys [
         {arriveTangent=-2.3163, interpMode=RCIM_Cubic, time=0, value=2}
         {arriveTangent=-1.9019, interpMode=RCIM_Cubic, leaveTangent=-1.9019, tangentMode=RCTM_User, time=1, value=0.4}
        ]}}})
    Index 1 Param 2=0.4
    Index 1 Param 3=0.6
    Index 1 Param 4=1
    Index 2 Param 1=&FloatFromCurve(FloatCurve=DataInterface() {propertyValues {Curve {keys [<at 0.46 0>,<at 1 1>]}}})
    Index 2 Param 2=45
    Module.Param0WriteEnabled=true
    Module.Param0WriteEnabled1=false
    Module.Param0WriteEnabled2=true
    Module.Param1WriteEnabled=true
    Module.Param1WriteEnabled1=true
    Module.Param1WriteEnabled2=true
    Module.Param2WriteEnabled=true
    Module.Param2WriteEnabled1=true
    Module.Param2WriteEnabled2=true
    Module.Param3WriteEnabled=true
    Module.Param3WriteEnabled1=true
    Module.Param3WriteEnabled2=true
    Write Parameter Index 1=true
    Write Parameter Index 2=true
   )
   &ScaleColor(
    Scale Alpha=&Multiply_Float(
     A=&FloatFromCurve(
      FloatCurve=DataInterface() {propertyValues {Curve {keys [
          {interpMode=RCIM_Cubic, tangentMode=RCTM_User, time=0, value=0}
          {arriveTangent=-7.8153, interpMode=RCIM_Cubic, leaveTangent=-7.8153, tangentMode=RCTM_User, time=0.5, value=1}
          {tangentMode=RCTM_User, time=1, value=0}
         ]}}}
      Scale Curve=&MakeCustomFloatFromBool(Bool=$System.IsAltCostume, True Float=0.45, False Float=0.5)
     )
     B=&FNTimeOfDayScalar(Night Scale=0.25, Morning Scale=0.5, Day Scale=0.5, Evening Scale=0.5)
    )
    ScaleA=true
    ScaleRGB=true
   )
   &ScaleMeshSize(Scale Factor=&VectorFromCurve(
     VectorCurve=DataInterface() {propertyValues {
       XCurve {keys [<at 0 0.5 RCIM_Cubic>,<at 1 1 RCIM_Cubic>]}
       YCurve {keys [<at 0 0.5 RCIM_Cubic>,<at 1 1 RCIM_Cubic>]}
       ZCurve {keys [<at 0 0.7 RCIM_Cubic>,<at 1 1 RCIM_Cubic>]}
      }}
     Scale Curve=&VectorFromFloat(Value=1)
    ))
   &FN_Mesh_Attributes(
    Color=&MultiplyLinearColor(Linear Color=$Particles.Color, Scale Factor=<1 1 1 0.25>)
    Create Color Attribute=true
    Create Scale Attribute=true
    Scale=&Multiply_VectorByFloat(Vector=$Particles.Scale, Float=2)
    Write to Module Namespace=true
   )
  ]
  rendererClasses [&NiagaraMeshRendererProperties]
  renderers [
   {rendererClass=&NiagaraMeshRendererProperties, rendererIndex=0}
   {rendererClass=&NiagaraMeshRendererProperties, rendererIndex=1}
  ]
 }
 Airglows {
  emitterUpdateScript [
   &EmitterState(Inactive Response=::NewEnumerator1, Life Cycle Mode=::NewEnumerator1, Loop Duration=0.1, UseLoopDelay=false)
   &SpawnBurst_Instantaneous(Age:NiagaraFloat=$Emitter.Age, Spawn Count=4, Use Spawn Probability=false, UseLoopCountLimit=false)
  ]
  particleSpawnScript [
   &InitializeParticle(
    Color=&HueShiftLinearColor(
     Hue=0
     Saturation=0.75
     Value=1
     Linear Color=&MultiplyLinearColor(
      Linear Color=$System.ColorPlusBoost
      Scale Factor=&MakeFloat4FromVectorAndFloat("Vector (RGB)"=&VectorFromFloat(Value=1.5), "Float (Alpha)"=0.9)
     )
    )
    Lifetime=5
    Sprite Size Mode=::NewEnumerator3
    Uniform Sprite Size=5
    UsePositionOffset=false
    Write Material Random=true
   )
   &CameraOffset(Camera Offset Amount=-15)
  ]
  particleUpdateScript [
   &ParticleState(
    Kill Particles When Lifetime Has Elapsed=false
    Let Infinitely Lived Particles Die When Emitter Deactivates=true
    Loop Particles Lifetime=true
   )
   &SkeletalMeshLocation(
    Bone / Socket Index=&ReturnExecIndex()
    Bone Sampling Mode=::NewEnumerator4
    MaskSpawnGroup=false
    Skeletal Mesh=$User.SkeletalMesh
    Write To Particle Position=false
   )
   /SetVariables(Particles.Velocity=&ReturnMeshOrientationAxis(Mesh Orientation Axis=::XAxis, Mesh Orientation Quat=$Particles.SkeletalMeshLocation.SampledRotation)) {enabled=false, isSetParameters=true}
   /Niagara/Modules/Debug/SpriteBasedLine(
    Color:LinearColor=<rgba 1 0 0.2237 1>
    Point A:NiagaraPosition=&Add_Vector(
     A=Linked() {linkedVariable {name=Particles.SkeletalMeshLocation.SampledPosition, type=NiagaraPosition}}
     B=&Multiply_VectorByFloat(
      Vector=&ReturnMeshOrientationAxis(Mesh Orientation Axis=::ZAxis, Mesh Orientation Quat=$Particles.SkeletalMeshLocation.SampledRotation)
      Float=&SelectFloatFromArray(
       Direct Array Index=&ReturnExecIndex()
       Float Selection Array=DataInterface() {propertyValues {FloatData [50,-50,30,-30]}}
      )
     )
    )
    Point B:NiagaraPosition=&Add_Vector(
     A=Linked() {linkedVariable {name=Particles.SkeletalMeshLocation.SampledPosition, type=NiagaraPosition}}
     B=&Multiply_VectorByFloat(
      Vector=&ReturnMeshOrientationAxis(Mesh Orientation Axis=::ZAxis, Mesh Orientation Quat=$Particles.SkeletalMeshLocation.SampledRotation)
      Float=&SelectFloatFromArray(
       Direct Array Index=&ReturnExecIndex()
       Float Selection Array=DataInterface() {propertyValues {FloatData [-150,150,-100,100]}}
      )
     )
    )
    SpriteWidth:NiagaraFloat=&SelectFloatFromArray(
     Direct Array Index=&ReturnExecIndex()
     Float Selection Array=DataInterface() {propertyValues {FloatData [35,35,35,35]}}
    )
   )
   /SetVariables(Particles.Position=&ConvertVectorToPosition(
     Input Position=Linked() {linkedVariable {name=Particles.SpriteBasedLine.Position, type=NiagaraPosition}}
     Vector Is In Absolute Large World Space:NiagaraBool=false
    )) {isSetParameters=true}
   &ScaleColor(
    Scale Alpha=&Multiply_Float(
     A=$System.Timeline.NormalizedTime
     B=&MakeCustomFloatFromBool(Bool=$System.IsAltCostume, True Float=0.65, False Float=1)
    )
    ScaleA=true
    ScaleRGB=true
   )
  ]
  rendererClasses [&NiagaraSpriteRendererProperties]
  renderers [{rendererClass=&NiagaraSpriteRendererProperties}]
 }
 Impact_Flare_and_Corona {
  emitterUpdateScript [
   &EmitterState()
   &SpawnRate(SpawnRate=8, Use Spawn Probability=false)
  ]
  particleSpawnScript [
   &InitializeParticle(
    Color=&MultiplyLinearColor(
     Linear Color=$System.ColorPlusBoost
     Scale Factor=&MakeFloat4FromVectorAndFloat("Vector (RGB)"=&VectorFromFloat(Value=7), "Float (Alpha)"=0.5)
    )
    Lifetime Max=1.2
    Lifetime Min=0.8
    Lifetime Mode=::NewEnumerator1
    Position Mode=::NewEnumerator1
    Sprite Rotation Mode=::NewEnumerator2
    Sprite Size=<80 100>
    Sprite Size Mode=::NewEnumerator4
    UsePositionOffset=false
    Write Material Random=true
   )
   &CameraOffset(Camera Offset Amount=5)
  ]
  particleUpdateScript [
   &ParticleState()
   &DynamicMaterialParameters(
    Index 0 Param 1=&FloatFromCurve(FloatCurve=DataInterface() {propertyValues {Curve {keys [<at 0 1 RCIM_Cubic>,<at 0.8 0 RCIM_Cubic>]}}})
    Index 0 Param 2=2
    Index 0 Param 3=1
    Index 0 Param 4=1
    Index 1 Param 1=&FloatFromCurve(FloatCurve=DataInterface() {propertyValues {Curve {keys [
         {arriveTangent=-2.3163, interpMode=RCIM_Cubic, time=0, value=2}
         {arriveTangent=-1.9019, interpMode=RCIM_Cubic, leaveTangent=-1.9019, tangentMode=RCTM_User, time=1, value=0.4}
        ]}}})
    Index 1 Param 2=0.4
    Index 1 Param 3=0.6
    Index 1 Param 4=1
    Index 2 Param 1=&FloatFromCurve(FloatCurve=DataInterface() {propertyValues {Curve {keys [<at 0.46 0>,<at 1 1>]}}})
    Index 2 Param 2=45
    Module.Param0WriteEnabled=true
    Module.Param0WriteEnabled1=false
    Module.Param0WriteEnabled2=true
    Module.Param1WriteEnabled=true
    Module.Param1WriteEnabled1=true
    Module.Param1WriteEnabled2=true
    Module.Param2WriteEnabled=true
    Module.Param2WriteEnabled1=true
    Module.Param2WriteEnabled2=true
    Module.Param3WriteEnabled=true
    Module.Param3WriteEnabled1=true
    Module.Param3WriteEnabled2=true
    Write Parameter Index 1=true
    Write Parameter Index 2=true
   )
   /SetVariables(Particles.CoronaSpriteSize=&Vector2DFromFloat(Value=&FloatFromCurve(
      FloatCurve=DataInterface() {propertyValues {Curve {keys [{interpMode=RCIM_Cubic, tangentMode=RCTM_User, time=0, value=0}, {interpMode=RCIM_Cubic, tangentMode=RCTM_User, time=1, value=1}]}}}
      Scale Curve=100
     ))) {isSetParameters=true}
   &ScaleColor(
    Scale Alpha=&Multiply_Float(
     A=&FloatFromCurve(
      FloatCurve=DataInterface() {propertyValues {Curve {keys [
          {arriveTangent=3.1838, interpMode=RCIM_Cubic, leaveTangent=3.1838, tangentMode=RCTM_User, time=0, value=0}
          {interpMode=RCIM_Cubic, tangentMode=RCTM_User, time=0.5, value=1}
          {arriveTangent=-3.1838, interpMode=RCIM_Cubic, leaveTangent=-3.1838, tangentMode=RCTM_User, time=1, value=0}
         ]}}}
      Scale Curve=&MakeCustomFloatFromBool(Bool=$System.IsAltCostume, True Float=0.75, False Float=1)
     )
     B=&FNTimeOfDayScalar(Night Scale=0.2, Morning Scale=0.25, Day Scale=0.25, Evening Scale=0.25)
    )
    ScaleA=true
    ScaleRGB=true
   )
   &FN_Light_Attributes(
    Color=$System.ColorPlusBoost
    Create Color Attribute=true
    Create Exponent Attribute=true
    Create Position Attribute=true
    Create Radius Attribute=true
    Exponent=6
    Position=<0 0 15>
    Radius=300
    Scale Color=&FloatFromCurve(
     FloatCurve=DataInterface() {propertyValues {Curve {keys [
         {interpMode=RCIM_Cubic, tangentMode=RCTM_User, time=0, value=0}
         {interpMode=RCIM_Cubic, tangentMode=RCTM_User, time=0.5, value=1}
         {tangentMode=RCTM_User, time=1, value=0}
        ]}}}
     Scale Curve=1.5
    )
   )
   &ScaleSpriteSize(
    Uniform Curve Scale=&MakeCustomFloatFromBool(Bool=$System.IsAltCostume, True Float=0.8, False Float=1)
    Uniform Curve Sprite Scale=DataInterface() {propertyValues {Curve {keys [
        {interpMode=RCIM_Cubic, tangentMode=RCTM_User, time=0, value=0.5}
        {interpMode=RCIM_Cubic, tangentMode=RCTM_User, time=0.5, value=1}
        {tangentMode=RCTM_User, time=1, value=0.5}
       ]}}}
   )
   &SkeletalMeshLocation(
    Bone / Socket Index=&ModuloInt(A=&ReturnParticleIDAttributes(), B=4)
    Bone Sampling Mode=::NewEnumerator4
    MaskSpawnGroup=false
    Skeletal Mesh=$User.SkeletalMesh
    Use Sampled Mesh Attribute as Ribbon ID=false
   )
  ]
  rendererClasses [&NiagaraSpriteRendererProperties, &NiagaraLightRendererProperties]
  renderers [
   {rendererClass=&NiagaraSpriteRendererProperties}
   {rendererClass=&NiagaraSpriteRendererProperties, rendererIndex=1}
   {rendererClass=&NiagaraLightRendererProperties, rendererIndex=2}
  ]
 }
 LargeArcs002 {
  emitterSpawnScript [/SetVariables(Emitter.Num Particles=4) {isSetParameters=true}]
  emitterUpdateScript [
   &EmitterState()
   &SpawnBurst_Instantaneous(Spawn Count=$Emitter.Num Particles, Use Spawn Probability=false, UseLoopCountLimit=false)
   /SetVariables(
    Emitter.Random Bone:NiagaraInt32=&RandomRangeInt(Minimum=0, Maximum=3)
    Emitter.random Float:NiagaraFloat=&RandomRangeFloat(Minimum=0, Maximum=1)
    Emitter.random bool:NiagaraBool=&SetBoolByFloatComparison(A=&RandomRangeFloat(Minimum=0, Maximum=1, Evaluation Type=::NewEnumerator1), B=0.75)
    Emitter.random float 2:NiagaraFloat=&RandomRangeFloat(Minimum=0.65, Maximum=1, Evaluation Type=::NewEnumerator1)
    Emitter.random life:NiagaraFloat=&RandomRangeFloat(Minimum=0, Maximum=1, Evaluation Type=::NewEnumerator1)
    Emitter.random sine:NiagaraFloat=&RandomRangeFloat(Minimum=0.1, Maximum=2, Evaluation Type=::NewEnumerator1)
   ) {isSetParameters=true, module=SetVariables_#3}
  ]
  particleSpawnScript [
   /SetVariables(
    Particles.Get Random Bool:NiagaraBool=$Emitter.random bool
    Particles.get random float 2:NiagaraFloat=$Emitter.random float 2
    Particles.get random life:NiagaraFloat=$Emitter.random life
   ) {isSetParameters=true}
   &InitializeRibbon(
    Color=&MultiplyLinearColor(
     Linear Color=$System.ColorPlusBoost
     Scale Factor=&MakeFloat4FromVectorAndFloat("Vector (RGB)"=&VectorFromFloat(Value=7.5), "Float (Alpha)"=1)
    )
    Lifetime=&MakeCustomFloatFromBool(Bool=true, True Float=1, False Float=&RandomRangeFloat(Minimum=0, Maximum=1))
    Ribbon Width=20
    UsePositionOffset=false
    Write Material Random=true
   )
   /SetVariables(Particles.SpriteSize=<5 5>) {isSetParameters=true}
   &FN_SetSpawnGroup(
    Normalize Spawn Group=true
    Particles Per Group:NiagaraInt32=$Emitter.Num Particles
    Pattern=::NewEnumerator0
    Set Ribbon ID and Ribbon Link Order=true
   )
   /SetVariables(Particles.Get random float:NiagaraFloat=$Emitter.random Float, Particles.get random bone:NiagaraInt32=$Emitter.Random Bone) {isSetParameters=true, module=SetVariables_#4}
   /SetVariables(Particles.invert float:NiagaraFloat=&SelectFloatFromArray(Direct Array Index=$Particles.get random bone, Float Selection Array=DataInterface() {propertyValues {FloatData [-1,1,-1,1]}})) {isSetParameters=true, module=SetVariables_#5}
   /SetVariables(Particles.Random Sine:NiagaraFloat=$Emitter.random sine) {isSetParameters=true}
  ]
  particleUpdateScript [
   &ParticleState()
   &SkeletalMeshLocation(
    Bone / Socket Index=$Particles.get random bone
    Bone Sampling Mode=::NewEnumerator4
    MaskSpawnGroup=false
    Orientation Sampling=::NewEnumerator1
    Skeletal Mesh=$User.SkeletalMesh
    Use Sampled Mesh Attribute as Ribbon ID=false
    Write To Particle Position=false
   )
   &ScaleColor(
    Scale Alpha=&Multiply_Float(
     A=&MakeCustomFloatFromBool(
      Bool=$Particles.Initial.Get Random Bool
      True Float=&Sine(Normalized Angle=$Emitter.Age, Period=$Particles.Initial.Get random float, Scale=1, Bias=0)
      False Float=&Sine(Normalized Angle=$Emitter.Age, Period=&RandomRangeFloat(Minimum=0.1, Maximum=1), Scale=1, Bias=0)
     )
     B=&MakeCustomFloatFromBool(Bool=$System.IsAltCostume, True Float=0.8, False Float=1)
    )
    ScaleA=true
    ScaleRGB=true
   )
   &ScaleRibbonWidth(Ribbon Width Scale=&Multiply_Float(
     A=&FloatFromCurve(FloatCurve=DataInterface() {propertyValues {Curve {keys [
          {interpMode=RCIM_Cubic, tangentMode=RCTM_User, time=0, value=0.1}
          {interpMode=RCIM_Cubic, tangentMode=RCTM_User, time=0.25, value=1}
          {arriveTangentWeight=0.1491, interpMode=RCIM_Cubic, leaveTangentWeight=0.2236, tangentMode=RCTM_User, time=0.75, value=1}
          {arriveTangentWeight=0.1, interpMode=RCIM_Cubic, tangentMode=RCTM_User, time=1, value=0.1}
         ]}}})
     B=&FloatFromCurve(
      FloatCurve=DataInterface() {propertyValues {Curve {keys [
          {arriveTangent=6.6108, interpMode=RCIM_Cubic, leaveTangent=6.6108, tangentMode=RCTM_User, time=0.05, value=0}
          {arriveTangentWeight=0.48, interpMode=RCIM_Cubic, leaveTangentWeight=0.48, time=0.5, value=1}
          {arriveTangent=-6.3941, interpMode=RCIM_Cubic, leaveTangent=-6.3941, tangentMode=RCTM_User, time=0.95, value=0}
         ]}}}
      CurveIndex=$Particles.Initial.Normalized Spawn Group
     )
    )) {module=ScaleRibbonWidth001}
   &CylinderLocation(
    Cylinder Axis:Vector3f=&ReturnMeshOrientationAxis(Mesh Orientation Axis=::ZAxis, Mesh Orientation Quat=$Particles.SkeletalMeshLocation.SampledRotation)
    Cylinder Distribution:ENiagara_CylinderMode=::NewEnumerator2
    Cylinder Height:NiagaraFloat=&Multiply_Float(
     A=$Particles.get random float 2
     B=&SelectFloatFromArray(Direct Array Index=$Particles.get random bone, Float Selection Array=DataInterface() {propertyValues {FloatData [75,75,45,45]}})
    )
    Cylinder Height Midpoint=0
    Cylinder Origin:NiagaraPosition=&ConvertVectorToPosition(
     Input Position=Linked() {linkedVariable {name=Particles.SkeletalMeshLocation.SampledPosition, type=NiagaraPosition}}
     Vector Is In Absolute Large World Space:NiagaraBool=false
    )
    Cylinder Radius=&FloatFromCurve(
     FloatCurve=DataInterface() {propertyValues {Curve {keys [
         {
          arriveTangent=5.1386
          arriveTangentWeight=1.1087
          interpMode=RCIM_Cubic
          leaveTangent=5.1386
          leaveTangentWeight=1.1087
          tangentMode=RCTM_User
          tangentWeightMode=RCTWM_WeightedBoth
          time=0
          value=0
         }
         {
          arriveTangentWeight=0.1963
          interpMode=RCIM_Cubic
          leaveTangentWeight=0.1963
          tangentMode=RCTM_User
          tangentWeightMode=RCTWM_WeightedBoth
          time=1
          value=1
         }
        ]}}}
     CurveIndex=$Particles.Normalized Spawn Group
     Scale Curve=&SelectFloatFromArray(Direct Array Index=$Particles.get random bone, Float Selection Array=DataInterface() {propertyValues {FloatData [10,10,7,7]}})
    )
    MaskSpawnGroup=false
    Non Uniform Scale:Vector3f=&Multiply_VectorByFloat(Vector=<0.25 1 1>, Float=$Particles.Initial.invert float)
    Rotation Mode:ENiagara_RotationMode=::NewEnumerator1
    U Position:NiagaraFloat=&Multiply_Float(A=$Particles.Normalized Spawn Group, B=&RandomRangeFloat(Minimum=0, Maximum=3))
    UseOwnerScale=true
    V Position:NiagaraFloat=$Particles.Normalized Spawn Group
   )
  ]
  rendererClasses [&NiagaraRibbonRendererProperties, &NiagaraSpriteRendererProperties]
  renderers [
   {rendererClass=&NiagaraRibbonRendererProperties}
   {rendererClass=&NiagaraRibbonRendererProperties, rendererIndex=1}
   {rendererClass=&NiagaraSpriteRendererProperties, rendererIndex=2}
  ]
 }
 Light {
  emitterUpdateScript [
   &EmitterState()
   &SpawnBurst_Instantaneous(Age:NiagaraFloat=$Emitter.Age, Spawn Count=4, Use Spawn Probability=false, UseLoopCountLimit=false)
  ]
  particleSpawnScript [&InitializeParticle(
    Color=&MultiplyLinearColor(
     Linear Color=$System.ColorPlusBoost
     Scale Factor=&MakeFloat4FromVectorAndFloat("Vector (RGB)"=&VectorFromFloat(Value=0.75), "Float (Alpha)"=0.1)
    )
    Lifetime=5
    UsePositionOffset=false
    Write Material Random=true
   )]
  particleUpdateScript [
   &ParticleState(
    Kill Particles When Lifetime Has Elapsed=false
    Let Infinitely Lived Particles Die When Emitter Deactivates=true
    Loop Particles Lifetime=true
   )
   &SkeletalMeshLocation(
    Bone / Socket Index=&ReturnExecIndex()
    Bone Sampling Mode=::NewEnumerator4
    MaskSpawnGroup=false
    Sampled Position Offset=&SelectVectorFromArray(
     Direct Array Index=&ReturnExecIndex()
     Vector Selection Array=DataInterface() {propertyValues {FloatData [<0 0 0>, <0 0 0>, <0 0 0>, <0 0 0>]}}
    )
    Skeletal Mesh=$User.SkeletalMesh
   )
   &ScaleColor(
    Scale Alpha=&Waveform(
     Waveform Count=::NewEnumerator2
     Global Phase=&Multiply_Float(A=$Emitter.Age, B=3)
     Global Amplitude Scale=120
     "Waveform [1]"=::NewEnumerator1
     "[1] Frequency"=1
     "[1] Amplitude Min/Max"=<0 1>
     "[2] Average"=false
     "[2] Input Position"=0
     "[2] Phase"=$Emitter.Age
     "[2] Phase Scale"=1
     "[2] Compound Frequency"=2
     "[2] Frequency"=2
     "[2] Amplitude Min/Max"=<0 1>
     "[2] Scale"=1
     "[2] Bias"=0
     "[2] Exponent"=1
     "[2] Pulse Duration"=0.1
     "[2] Clip Triangle Min/Max"=<0 0>
     "[2] Random Seed"=0
     "[3] Average"=false
     "[3] Input Position"=0
     "[3] Phase"=$Emitter.Age
     "[3] Phase Scale"=1
     "[3] Frequency"=4
     "[3] Compound Frequency"=2
     "[3] Amplitude Min/Max"=<0 1>
     "[3] Scale"=1
     "[3] Bias"=0
     "[3] Exponent"=1
     "[3] Pulse Duration"=0.1
     "[3] Clip Triangle Min/Max"=<0 0>
     "[3] Random Seed"=0
    )
    ScaleA=true
    ScaleRGB=true
   )
  ]
  rendererClasses [&NiagaraLightRendererProperties]
  renderers [{rendererClass=&NiagaraLightRendererProperties, rendererIndex=0}]
 }
 Lightning_on_ring {
  emitterUpdateScript [
   &EmitterState(Scalability Mode=::NewEnumerator1)
   &SpawnBurst_Instantaneous(
    Spawn Count=&RandomRangeInt(Minimum=2, Maximum=3)
    Use Spawn Probability=false
    UseLoopCountLimit=false
   )
   /SetVariables(
    Emitter.Random Bone:NiagaraInt32=&RandomRangeInt(Minimum=0, Maximum=3)
    Emitter.random Float:NiagaraFloat=&RandomRangeFloat(Minimum=0, Maximum=1)
    Emitter.random sine:NiagaraFloat=&RandomRangeFloat(Minimum=0.1, Maximum=2, Evaluation Type=::NewEnumerator1)
   ) {isSetParameters=true, module=SetVariables_#3}
  ]
  particleSpawnScript [
   &InitializeParticle(
    Color=&MultiplyLinearColor(
     Linear Color=$System.ColorPlusBoost
     Scale Factor=&MakeFloat4FromVectorAndFloat("Vector (RGB)"=&VectorFromFloat(Value=25), "Float (Alpha)"=1)
    )
    Lifetime Max=0.18
    Lifetime Min=0.1
    Lifetime Mode=::NewEnumerator1
    Sprite Size Max=<35 25>
    Sprite Size Min=<25 15>
    Sprite Size Mode=::NewEnumerator2
    Sprite UV Mode=::NewEnumerator1
    UsePositionOffset=false
    Write Material Random=true
   )
   &DynamicMaterialParameters(
    Index 0 Param 1=&RandomRangeFloat(Minimum=1, Maximum=500)
    Index 0 Param 2=&RandomRangeFloat(Minimum=1, Maximum=2)
    Module.Param0WriteEnabled=true
    Module.Param1WriteEnabled=true
    Module.Param2WriteEnabled=true
    Module.Param3WriteEnabled=true
   )
   /SetVariables(Particles.Get random float:NiagaraFloat=$Emitter.random Float, Particles.get random bone:NiagaraInt32=$Emitter.Random Bone) {isSetParameters=true, module=SetVariables_#4}
   /SetVariables(Particles.invert float:NiagaraFloat=&SelectFloatFromArray(Direct Array Index=$Particles.get random bone, Float Selection Array=DataInterface() {propertyValues {FloatData [-1,1,-1,1]}})) {isSetParameters=true, module=SetVariables_#5}
   &SpriteFacingAndAlignment(
    Sprite Alignment=&RandomVector(Vector Scale=1)
    Sprite Facing=&RandomVector(Vector Scale=1)
    WriteAlignment=true
    WriteFacing=true
   )
  ]
  particleUpdateScript [
   &SolveForcesAndVelocity(Clamp Velocity=false, Limit Acceleration=false)
   &ParticleState()
   &SkeletalMeshLocation(
    Bone / Socket Index=$Particles.get random bone
    Bone Sampling Mode=::NewEnumerator4
    MaskSpawnGroup=false
    Skeletal Mesh=$User.SkeletalMesh
    Use Sampled Mesh Attribute as Ribbon ID=false
    Write To Particle Position=false
   )
   &ScaleSpriteSize(
    Uniform Curve Scale=0.6
    Uniform Curve Sprite Scale=DataInterface() {propertyValues {Curve {keys [{tangentMode=RCTM_User, time=0, value=0.8}, {tangentMode=RCTM_User, time=1, value=1}]}}}
   )
   &ShapeLocation(
    Cylinder Height=&SelectFloatFromArray(Direct Array Index=$Particles.get random bone, Float Selection Array=DataInterface() {propertyValues {FloatData [75,75,45,45]}})
    Cylinder Height Midpoint=0
    Cylinder Radius=1
    MaskSpawnGroup=false
    Override Seed=false
    Rotation Axis=&Multiply_VectorByFloat(
     Vector=&ReturnMeshOrientationAxis(Mesh Orientation Axis=::ZAxis, Mesh Orientation Quat=$Particles.MeshOrientation)
     Float=$Particles.Initial.invert float
    )
    Rotation Mode=::NewEnumerator1
    Shape Origin=$Particles.SkeletalMeshLocation.SampledPosition
    Shape Primitive=::NewEnumerator1
    Surface Only=false
    UseOwnerScale=true
   )
  ]
  rendererClasses [&NiagaraSpriteRendererProperties]
  renderers [{rendererClass=&NiagaraSpriteRendererProperties}]
 }
 Lightning_on_ring001 {
  emitterUpdateScript [
   &EmitterState(
    Life Cycle Mode=::NewEnumerator1
    Loop Duration=&RandomRangeFloat(Minimum=0.1, Maximum=3)
    Recalculate Duration Each Loop=true
    UseLoopDelay=false
   )
   /SetVariables() {isSetParameters=true}
   &SpawnRate(SpawnRate=30, Use Spawn Probability=false)
  ]
  particleSpawnScript [
   &InitializeParticle(
    Color=&MultiplyLinearColor(
     Linear Color=$System.ColorPlusBoost
     Scale Factor=&MakeFloat4FromVectorAndFloat("Vector (RGB)"=&VectorFromFloat(Value=75), "Float (Alpha)"=1)
    )
    Lifetime Max=0.18
    Lifetime Min=0.1
    Lifetime Mode=::NewEnumerator1
    Sprite Size Max=<35 25>
    Sprite Size Min=<25 15>
    Sprite Size Mode=::NewEnumerator2
    Sprite UV Mode=::NewEnumerator1
    UsePositionOffset=false
    Write Material Random=true
   )
   &DynamicMaterialParameters(
    Index 0 Param 1=&RandomRangeFloat(Minimum=1, Maximum=500)
    Index 0 Param 2=&RandomRangeFloat(Minimum=1, Maximum=2)
    Module.Param0WriteEnabled=true
    Module.Param1WriteEnabled=true
    Module.Param2WriteEnabled=true
    Module.Param3WriteEnabled=true
   )
   &SpriteFacingAndAlignment(
    Sprite Alignment=&RandomVector(Vector Scale=1)
    Sprite Facing=&RandomVector(Vector Scale=1)
    WriteAlignment=true
    WriteFacing=true
   )
  ]
  particleUpdateScript [
   &SolveForcesAndVelocity(Clamp Velocity=false, Limit Acceleration=false)
   &ParticleState()
   &SkeletalMeshLocation(
    Bone / Socket Index=&RandomRangeInt(Minimum=0, Maximum=3)
    Bone Sampling Mode=::NewEnumerator4
    MaskSpawnGroup=false
    Skeletal Mesh=$User.SkeletalMesh
    Use Sampled Mesh Attribute as Ribbon ID=false
   )
   &ScaleSpriteSize(
    Uniform Curve Scale=0.7
    Uniform Curve Sprite Scale=DataInterface() {propertyValues {Curve {keys [{tangentMode=RCTM_User, time=0, value=0.8}, {tangentMode=RCTM_User, time=1, value=1}]}}}
   )
   &ScaleColor(
    Scale Alpha=&MakeCustomFloatFromBool(Bool=$System.IsAltCostume, True Float=0.8, False Float=1)
    ScaleA=true
    ScaleRGB=true
   )
  ]
  rendererClasses [&NiagaraSpriteRendererProperties]
  renderers [{rendererClass=&NiagaraSpriteRendererProperties}]
 }
}
summary {
 systemName=NS_CyberArmor_Female_Glider
 emitters [
  {emitterName=Lightning_on_ring, rendererClasses [&NiagaraSpriteRendererProperties]}
  {emitterName=LargeArcs002, rendererClasses [&NiagaraRibbonRendererProperties, &NiagaraSpriteRendererProperties]}
  {emitterName=Light, rendererClasses [&NiagaraLightRendererProperties]}
  {emitterName=Airglows, rendererClasses [&NiagaraSpriteRendererProperties]}
  {
   emitterName=Impact_Flare_and_Corona
   rendererClasses [&NiagaraSpriteRendererProperties, &NiagaraLightRendererProperties]
  }
  {emitterName=3d_glow, rendererClasses [&NiagaraMeshRendererProperties]}
  {emitterName=Lightning_on_ring001, rendererClasses [&NiagaraSpriteRendererProperties]}
 ]
 userVariables [
  User.Color {type=LinearColor, description="", defaultValue {struct=LinearColor, value=<rgba 0.3438 0 0.0108 1>}}
  User.IsAltCostume {type=NiagaraFloat, defaultValue=0, description=""}
  User.SkeletalMesh {
   type=NiagaraDataInterfaceSkeletalMesh
   defaultValue {
    struct=NiagaraExt_VariableValue_DataInterface
    value {dataInterface=&self:NiagaraDataInterfaceSkeletalMesh_0}
   }
  }
 ]
}
