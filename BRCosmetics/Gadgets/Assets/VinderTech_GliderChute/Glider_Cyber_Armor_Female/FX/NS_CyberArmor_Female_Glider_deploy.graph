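# Niagara electrical deploy effect for the female Cyber Armor glider.
# LargeArcs002 bursts short-lived, tapered ribbon arcs with sine-flickered alpha, Lightning_on_ring spawns randomly oriented sprites around a small offset shape, and Light adds a waveform-flickered sprite and light.
# All three tint from System.ColorPlusBoost.
# The bone-attached LargeArcs and spike lightning mesh emitters are disabled, as is skeletal mesh sampling in the active emitters, leaving placement to shape modules and offsets.
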NiagaraSystem /BRCosmetics/Gadgets/Assets/VinderTech_GliderChute/Glider_Cyber_Armor_Female/FX/NS_CyberArmor_Female_Glider_deploy
dependencies {
 usedDataInterfaces [&NiagaraDataInterfaceArrayFloat, &NiagaraDataInterfaceCurve]
 usedDynamicInputs []
 usedModules [
  &SystemState
  &self:SystemSpawnScript.SystemScriptSource.SystemScriptGraph.NiagaraNodeAssignment_0.SetVariables_#1
 ]
}
emitters {
 LargeArcs {
  bEnabled=false
  emitterUpdateScript [
   &EmitterState()
   &SpawnBurst_Instantaneous(Spawn Count=8, Use Spawn Probability=false, UseLoopCountLimit=false)
   /SetVariables(
    Emitter.Random Bone:NiagaraInt32=&RandomRangeInt(Minimum=0, Maximum=3)
    Emitter.random Float:NiagaraFloat=&RandomRangeFloat(Minimum=0, Maximum=1)
   ) {isSetParameters=true, module=SetVariables_#2}
  ]
  particleSpawnScript [
   &InitializeRibbon(
    Color=<rgba 5 0.1698 0.8452 1>
    Lifetime=&UniformRangedFloat(Minimum=0.2, Maximum=0.35)
    Module.Write Color=true
    Module.Write Lifetime=true
    Module.Write Mass=true
    Module.Write Material Random=true
    Module.Write Position=false
    Module.Write RibbonTwist=false
    Module.Write RibbonWidth=true
    Ribbon Width=10
   )
   /SetVariables(
    Particles.RibbonID=/Niagara/DynamicInputs/NiagaraID/MakeNiagaraIDFromInt(Index:NiagaraInt32=$Emitter.LoopCount, Acquire Tag:NiagaraInt32=0)
    Particles.link order:NiagaraFloat=&MakeFloatFromInt(Int=&ReturnParticleIDAttributes())
   ) {isSetParameters=true}
   /SetVariables(Particles.SpriteSize=<5 5>) {isSetParameters=true, module=SetVariables_#3}
   &FN_SetSpawnGroup(Inclusivity=true, Normalize Spawn Group=true, Particles Per Group=8, Pattern=::NewEnumerator0, Set Ribbon ID and Ribbon Link Order=true)
  ]
  particleUpdateScript [
   &ParticleState()
   &SkeletalMeshLocation(
    Bone / Socket Index=$Emitter.Random Bone
    Bone Sampling Mode=::NewEnumerator4
    MaskSpawnGroup=false
    Sampled Position Offset=&Multiply_Vector(
     A=<1 1 1>
     B=&VectorFromFloat(Value=&SelectFloatFromArray(Direct Array Index=$Emitter.Random Bone, Float Selection Array=DataInterface() {propertyValues {FloatData [1,-1,-1,1]}}))
    )
    Skeletal Mesh=$User.SkeletalMesh
   )
   &ScaleColor(
    Scale Alpha=&FloatFromCurve(FloatCurve=DataInterface() {propertyValues {Curve {keys [
         {arriveTangent=8.0553, interpMode=RCIM_Cubic, leaveTangent=8.0553, tangentMode=RCTM_User, time=0, value=0}
         <at 0.2113 1 RCIM_Cubic>
         <at 0.4546 0>
         {arriveTangent=-8.0018, interpMode=RCIM_Cubic, leaveTangent=7.1228, tangentMode=RCTM_Break, time=0.5517, value=0}
         <at 0.7524 1 RCIM_Cubic>
         {arriveTangent=-10.4882, interpMode=RCIM_Cubic, leaveTangent=-10.4882, tangentMode=RCTM_User, time=1, value=0}
        ]}}})
    ScaleA=true
    ScaleRGB=true
   )
   &ScaleRibbonWidth(Ribbon Width Scale=&Multiply_Float(
     A=&FloatFromCurve(FloatCurve=DataInterface() {propertyValues {Curve {keys [
          {interpMode=RCIM_Cubic, tangentMode=RCTM_User, time=0, value=0}
          {interpMode=RCIM_Cubic, tangentMode=RCTM_User, time=0.25, value=1}
          {arriveTangentWeight=0.1491, interpMode=RCIM_Cubic, leaveTangentWeight=0.2236, tangentMode=RCTM_User, time=0.75, value=1}
          {arriveTangentWeight=0.1, interpMode=RCIM_Cubic, tangentMode=RCTM_User, time=1, value=0}
         ]}}})
     B=&FloatFromCurve(
      FloatCurve=DataInterface() {propertyValues {Curve {keys [
          {arriveTangent=6.6108, interpMode=RCIM_Cubic, leaveTangent=6.6108, tangentMode=RCTM_User, time=0.05, value=0}
          {interpMode=RCIM_Cubic, tangentMode=RCTM_User, time=0.5, value=1}
          {arriveTangent=-6.3941, interpMode=RCIM_Cubic, leaveTangent=-6.3941, tangentMode=RCTM_User, time=0.95, value=0}
         ]}}}
      CurveIndex=&ReturnNormalizedExecIndex()
     )
    )) {module=ScaleRibbonWidth001}
  ]
  rendererClasses [&NiagaraRibbonRendererProperties, &NiagaraSpriteRendererProperties]
  renderers [
   {rendererClass=&NiagaraRibbonRendererProperties}
   {rendererClass=&NiagaraSpriteRendererProperties, rendererIndex=1}
  ]
 }
 LargeArcs002 {
  emitterSpawnScript [/SetVariables(Emitter.Num Particles=4) {isSetParameters=true}]
  emitterUpdateScript [
   &EmitterState()
   &SpawnBurst_Instantaneous(Spawn Count=$Emitter.Num Particles, Spawn Time=0.01, Use Spawn Probability=false, UseLoopCountLimit=false)
   /SetVariables(
    Emitter.Random Bone:NiagaraInt32=&RandomRangeInt(Minimum=0, Maximum=3)
    Emitter.random Float:NiagaraFloat=&RandomRangeFloat(Minimum=0, Maximum=1)
    Emitter.random bool:NiagaraBool=&SetBoolByFloatComparison(A=&RandomRangeFloat(Minimum=0, Maximum=1, Evaluation Type=::NewEnumerator1), B=0.75)
    Emitter.random float 2:NiagaraFloat=&RandomRangeFloat(Minimum=0.65, Maximum=1, Evaluation Type=::NewEnumerator1)
    Emitter.random life:NiagaraFloat=&RandomRangeFloat(Minimum=0, Maximum=1, Evaluation Type=::NewEnumerator1)
    Emitter.random sine:NiagaraFloat=&RandomRangeFloat(Minimum=0.1, Maximum=2, Evaluation Type=::NewEnumerator1)
   ) {isSetParameters=true, module=SetVariables_#2}
   &ShapeLocation(
    MaskSpawnGroup=false
    Override Seed=false
    Ring Radius=&RandomRangeFloat(Minimum=0, Maximum=1)
    Shape Primitive=::NewEnumerator4
    UseOwnerScale=true
   ) {module=ShapeLocation001}
  ]
  particleSpawnScript [
   /SetVariables(
    Particles.Get Random Bool:NiagaraBool=$Emitter.random bool
    Particles.get random float 2:NiagaraFloat=$Emitter.random float 2
    Particles.get random life:NiagaraFloat=$Emitter.random life
   ) {isSetParameters=true}
   &InitializeRibbon(
    Color=&MultiplyLinearColor(
     Linear Color=$System.ColorPlusBoost
     Scale Factor=&MakeFloat4FromVectorAndFloat("Vector (RGB)"=&VectorFromFloat(Value=7.5), "Float (Alpha)"=1)
    )
    Lifetime=0.25
    Ribbon Width=15
    UsePositionOffset=false
    Write Material Random=true
   )
   &InitializeRibbon(
    Color=&MultiplyLinearColor(
     Linear Color=$User.Color
     Scale Factor=&MakeFloat4FromVectorAndFloat("Vector (RGB)"=&VectorFromFloat(Value=1), "Float (Alpha)"=1)
    )
    Lifetime=&UniformRangedFloat(Minimum=1, Maximum=1)
    Ribbon Width=1
    UsePositionOffset=false
    Write Material Random=true
   ) {enabled=false, module=InitializeRibbon001}
   /SetVariables(Particles.SpriteSize=<5 5>) {isSetParameters=true, module=SetVariables_#3}
   &FN_SetSpawnGroup(
    Inclusivity=true
    Normalize Spawn Group=true
    Particles Per Group:NiagaraInt32=$Emitter.Num Particles
    Pattern=::NewEnumerator0
    Set Ribbon ID and Ribbon Link Order=true
   )
   /SetVariables(Particles.Get random float:NiagaraFloat=$Emitter.random Float, Particles.get random bone:NiagaraInt32=$Emitter.Random Bone) {isSetParameters=true, module=SetVariables_#4}
   /SetVariables(Particles.invert float:NiagaraFloat=&SelectFloatFromArray(Direct Array Index=$Particles.get random bone, Float Selection Array=DataInterface() {propertyValues {FloatData [-1,1,-1,1]}})) {isSetParameters=true, module=SetVariables_#5}
   /SetVariables(Particles.Random Sine:NiagaraFloat=$Emitter.random sine) {isSetParameters=true}
  ]
  particleUpdateScript [
   &ParticleState()
   &SkeletalMeshLocation(
    Bone / Socket Index=$Particles.get random bone
    Bone Sampling Mode=::NewEnumerator4
    MaskSpawnGroup=false
    Orientation Sampling=::NewEnumerator1
    Skeletal Mesh=$User.SkeletalMesh
    Use Sampled Mesh Attribute as Ribbon ID=false
    Write To Particle Position=false
   ) {enabled=false}
   &ScaleColor(
    Scale Alpha=&MakeCustomFloatFromBool(
     Bool=$Particles.Initial.Get Random Bool
     True Float=&Sine(Normalized Angle=$Emitter.Age, Period=$Particles.Initial.Get random float, Scale=1, Bias=0)
     False Float=&Sine(Normalized Angle=$Emitter.Age, Period=&RandomRangeFloat(Minimum=0.1, Maximum=1), Scale=1, Bias=0)
    )
    ScaleA=true
    ScaleRGB=true
   )
   &ScaleRibbonWidth(Ribbon Width Scale=&Multiply_Float(
     A=&FloatFromCurve(FloatCurve=DataInterface() {propertyValues {Curve {keys [
          {interpMode=RCIM_Cubic, tangentMode=RCTM_User, time=0, value=0.1}
          {interpMode=RCIM_Cubic, tangentMode=RCTM_User, time=0.25, value=1}
          {arriveTangentWeight=0.1491, interpMode=RCIM_Cubic, leaveTangentWeight=0.2236, tangentMode=RCTM_User, time=0.75, value=1}
          {arriveTangentWeight=0.1, interpMode=RCIM_Cubic, tangentMode=RCTM_User, time=1, value=0.1}
         ]}}})
     B=&FloatFromCurve(
      FloatCurve=DataInterface() {propertyValues {Curve {keys [
          {arriveTangent=6.6108, interpMode=RCIM_Cubic, leaveTangent=6.6108, tangentMode=RCTM_User, time=0.05, value=0}
          {arriveTangentWeight=0.48, interpMode=RCIM_Cubic, leaveTangentWeight=0.48, time=0.5, value=1}
          {arriveTangent=-6.3941, interpMode=RCIM_Cubic, leaveTangent=-6.3941, tangentMode=RCTM_User, time=0.95, value=0}
         ]}}}
      CurveIndex=$Particles.Initial.Normalized Spawn Group
     )
    )) {module=ScaleRibbonWidth001}
   &CylinderLocation(
    Cylinder Axis:Vector3f=<0 0 1>
    Cylinder Distribution:ENiagara_CylinderMode=::NewEnumerator2
    Cylinder Height Midpoint=0
    Cylinder Radius=&FloatFromCurve(
     FloatCurve=DataInterface() {propertyValues {Curve {keys [
         {
          arriveTangent=5.1386
          arriveTangentWeight=1.1087
          interpMode=RCIM_Cubic
          leaveTangent=5.1386
          leaveTangentWeight=1.1087
          tangentMode=RCTM_User
          tangentWeightMode=RCTWM_WeightedBoth
          time=0
          value=0
         }
         {
          arriveTangentWeight=0.1963
          interpMode=RCIM_Cubic
          leaveTangentWeight=0.1963
          tangentMode=RCTM_User
          tangentWeightMode=RCTWM_WeightedBoth
          time=1
          value=1
         }
        ]}}}
     CurveIndex=$Particles.Normalized Spawn Group
    )
    MaskSpawnGroup=false
    Rotation Mode:ENiagara_RotationMode=::NewEnumerator1
    U Position:NiagaraFloat=&Multiply_Float(A=$Particles.Normalized Spawn Group, B=&RandomRangeFloat(Minimum=0, Maximum=3))
    UseOwnerScale=true
    V Position:NiagaraFloat=$Particles.Normalized Spawn Group
   )
   &ShapeLocation(
    Cylinder Distribution=::NewEnumerator2
    Cylinder Height=25
    Cylinder Height Midpoint=0
    Cylinder Radius=5
    MaskSpawnGroup=false
    Offset=<0 4 5>
    Offset Mode=::NewEnumerator0
    Rotation Axis=<0 -1 0>
    Rotation Mode=::NewEnumerator1
    Shape Primitive=::NewEnumerator1
    U Position=&RandomRangeFloat(Minimum=0, Maximum=1)
    UseOwnerScale=true
    V Position=$Particles.Initial.Normalized Spawn Group
   )
  ]
  rendererClasses [&NiagaraRibbonRendererProperties]
  renderers [
   {rendererClass=&NiagaraRibbonRendererProperties}
   {rendererClass=&NiagaraRibbonRendererProperties, rendererIndex=1}
  ]
 }
 Light {
  emitterUpdateScript [
   &EmitterState()
   &SpawnRate(SpawnRate=10, Use Spawn Probability=false)
   &SpawnBurst_Instantaneous(Age:NiagaraFloat=$Emitter.Age, Spawn Time=0.01, Use Spawn Probability=false, UseLoopCountLimit=false)
  ]
  particleSpawnScript [&InitializeParticle(
    Color=&HueShiftLinearColor(
     Hue=0
     Saturation=0.8
     Value=1
     Linear Color=&MultiplyLinearColor(
      Linear Color=$System.ColorPlusBoost
      Scale Factor=&MakeFloat4FromVectorAndFloat("Vector (RGB)"=&VectorFromFloat(Value=0.2), "Float (Alpha)"=3)
     )
    )
    Lifetime=0.1
    Position Offset=<0 -3 0>
    UsePositionOffset=true
    Write Material Random=true
   )]
  particleUpdateScript [
   &ParticleState()
   &ScaleColor(
    Scale Alpha=&Waveform(
     Waveform Count=::NewEnumerator2
     Global Phase=&Multiply_Float(A=$Emitter.Age, B=3)
     Global Amplitude Scale=100
     "Waveform [1]"=::NewEnumerator1
     "[1] Frequency"=1
     "[1] Amplitude Min/Max"=<0 1>
     "[2] Average"=false
     "[2] Input Position"=0
     "[2] Phase"=$Emitter.Age
     "[2] Phase Scale"=1
     "[2] Compound Frequency"=2
     "[2] Frequency"=2
     "[2] Amplitude Min/Max"=<0 1>
     "[2] Scale"=1
     "[2] Bias"=0
     "[2] Exponent"=1
     "[2] Pulse Duration"=0.1
     "[2] Clip Triangle Min/Max"=<0 0>
     "[2] Random Seed"=0
     "[3] Average"=false
     "[3] Input Position"=0
     "[3] Phase"=$Emitter.Age
     "[3] Phase Scale"=1
     "[3] Frequency"=4
     "[3] Compound Frequency"=2
     "[3] Amplitude Min/Max"=<0 1>
     "[3] Scale"=1
     "[3] Bias"=0
     "[3] Exponent"=1
     "[3] Pulse Duration"=0.1
     "[3] Clip Triangle Min/Max"=<0 0>
     "[3] Random Seed"=0
    )
    ScaleA=true
    ScaleRGB=true
   )
  ]
  rendererClasses [&NiagaraSpriteRendererProperties, &NiagaraLightRendererProperties]
  renderers [
   {rendererClass=&NiagaraSpriteRendererProperties}
   {rendererClass=&NiagaraLightRendererProperties, rendererIndex=1}
  ]
 }
 Lightning_on_ring {
  emitterUpdateScript [
   &EmitterState()
   /SetVariables(
    Emitter.Random Bone:NiagaraInt32=&RandomRangeInt(Minimum=0, Maximum=3)
    Emitter.random Float:NiagaraFloat=&RandomRangeFloat(Minimum=0, Maximum=1)
    Emitter.random sine:NiagaraFloat=&RandomRangeFloat(Minimum=0.1, Maximum=2, Evaluation Type=::NewEnumerator1)
   ) {isSetParameters=true, module=SetVariables_#2}
   &SpawnRate(SpawnRate=25, Use Spawn Probability=false)
  ]
  particleSpawnScript [
   &InitializeParticle(
    Color=&MultiplyLinearColor(
     Linear Color=$System.ColorPlusBoost
     Scale Factor=&MakeFloat4FromVectorAndFloat("Vector (RGB)"=&VectorFromFloat(Value=50), "Float (Alpha)"=3)
    )
    Lifetime Max=0.18
    Lifetime Min=0.1
    Lifetime Mode=::NewEnumerator1
    Sprite Size Max=<35 25>
    Sprite Size Min=<25 15>
    Sprite Size Mode=::NewEnumerator2
    Sprite UV Mode=::NewEnumerator1
    UsePositionOffset=false
    Write Material Random=true
   )
   &DynamicMaterialParameters(
    Index 0 Param 1=&RandomRangeFloat(Minimum=1, Maximum=500)
    Index 0 Param 2=&RandomRangeFloat(Minimum=1, Maximum=2)
    Module.Param0WriteEnabled=true
    Module.Param1WriteEnabled=true
    Module.Param2WriteEnabled=true
    Module.Param3WriteEnabled=true
   )
   /SetVariables(Particles.Get random float:NiagaraFloat=$Emitter.random Float, Particles.get random bone:NiagaraInt32=$Emitter.Random Bone) {isSetParameters=true, module=SetVariables_#4}
   /SetVariables(Particles.invert float:NiagaraFloat=&SelectFloatFromArray(Direct Array Index=$Particles.get random bone, Float Selection Array=DataInterface() {propertyValues {FloatData [-1,1,-1,1]}})) {isSetParameters=true, module=SetVariables_#5}
   &SpriteFacingAndAlignment(
    Sprite Alignment=&RandomVector(Vector Scale=1)
    Sprite Facing=&RandomVector(Vector Scale=1)
    WriteAlignment=true
    WriteFacing=true
   )
   &CameraOffset(Camera Offset Amount=5)
  ]
  particleUpdateScript [
   &SolveForcesAndVelocity(Clamp Velocity=false, Limit Acceleration=false)
   &ParticleState()
   &SkeletalMeshLocation(
    Bone / Socket Index=$Particles.get random bone
    Bone Sampling Mode=::NewEnumerator4
    MaskSpawnGroup=false
    Skeletal Mesh=$User.SkeletalMesh
    Use Sampled Mesh Attribute as Ribbon ID=false
    Write To Particle Position=false
   ) {enabled=false}
   &ScaleSpriteSize(
    Uniform Curve Scale=0.3
    Uniform Curve Sprite Scale=DataInterface() {propertyValues {Curve {keys [{tangentMode=RCTM_User, time=0, value=0.8}, {tangentMode=RCTM_User, time=1, value=1}]}}}
   )
   &ShapeLocation(
    MaskSpawnGroup=false
    Offset=<0 4 5>
    Offset Mode=::NewEnumerator0
    Override Seed=false
    Rotation Axis=&Multiply_VectorByFloat(
     Vector=&ReturnMeshOrientationAxis(Mesh Orientation Axis=::ZAxis, Mesh Orientation Quat=$Particles.MeshOrientation)
     Float=$Particles.Initial.invert float
    )
    Rotation Mode=::NewEnumerator1
    Sphere Radius=5
    UseOwnerScale=true
   )
  ]
  rendererClasses [&NiagaraSpriteRendererProperties]
  renderers [{rendererClass=&NiagaraSpriteRendererProperties}]
 }
 Lightning_ou_of_spikes {
  bEnabled=false
  emitterUpdateScript [
   &EmitterState()
   &SpawnBurst_Instantaneous(
    Spawn Count=&RandomRangeInt(Minimum=4, Maximum=7)
    Use Spawn Probability=false
    UseLoopCountLimit=false
   )
  ]
  particleSpawnScript [
   &InitializeParticle(
    Color=&MultiplyLinearColor(
     Linear Color=$User.Color
     Scale Factor=&MakeFloat4FromVectorAndFloat("Vector (RGB)"=&VectorFromFloat(Value=5), "Float (Alpha)"=1)
    )
    Lifetime Max=0.15
    Lifetime Min=0.05
    Lifetime Mode=::NewEnumerator1
    Mesh Scale Max=<2 5 5>
    Mesh Scale Min=<2.5 1 1>
    Mesh Scale Mode=::NewEnumerator2
    Position Mode=::NewEnumerator1
    Sprite Size Mode=::NewEnumerator3
    Uniform Sprite Size=2
    UsePositionOffset=false
    Write Material Random=true
   )
   &InitialMeshOrientation(
    Mesh Orientation Mode=::NewEnumerator7
    Rotation=&RandomRangeVector(Minimum=<0 0.5 0.5>, Maximum=<1 -0.5 -0.5>)
    Use Rotation Vector=true
   )
  ]
  particleUpdateScript [
   &ParticleState()
   &DynamicMaterialParameters(
    Index 0 Param 1=1
    Index 0 Param 2=&FloatFromCurve(FloatCurve=DataInterface() {propertyValues {Curve {keys [<at 0 1>,<at 1 0.5>]}}})
    Module.Param0WriteEnabled=true
    Module.Param1WriteEnabled=true
    Module.Param2WriteEnabled=true
    Module.Param3WriteEnabled=true
   )
   &ScaleMeshSize(Scale Factor=&MakeVector(
     X=&FloatFromCurve(FloatCurve=DataInterface() {propertyValues {Curve {keys [{arriveTangent=2.5639, interpMode=RCIM_Cubic, leaveTangent=2.5639, tangentMode=RCTM_User, time=0, value=0.25}, <at 1 1 RCIM_Cubic>]}}})
     Y=1
     Z=1
    ))
  ]
  rendererClasses [&NiagaraMeshRendererProperties]
  renderers [{rendererClass=&NiagaraMeshRendererProperties, rendererIndex=0}]
 }
}
summary {
 systemName=NS_CyberArmor_Female_Glider_deploy
 emitters [
  {
   bEnabled=false
   emitterName=LargeArcs
   rendererClasses [&NiagaraRibbonRendererProperties, &NiagaraSpriteRendererProperties]
  }
  {bEnabled=false, emitterName=Lightning_ou_of_spikes, rendererClasses [&NiagaraMeshRendererProperties]}
  {emitterName=LargeArcs002, rendererClasses [&NiagaraRibbonRendererProperties]}
  {emitterName=Lightning_on_ring, rendererClasses [&NiagaraSpriteRendererProperties]}
  {emitterName=Light, rendererClasses [&NiagaraSpriteRendererProperties, &NiagaraLightRendererProperties]}
 ]
 userVariables [
  User.Color {type=LinearColor, description="", defaultValue {struct=LinearColor, value=<rgba 0.3438 0 0.0108 1>}}
  User.SkeletalMesh {
   type=NiagaraDataInterfaceSkeletalMesh
   defaultValue {
    struct=NiagaraExt_VariableValue_DataInterface
    value {dataInterface=&self:NiagaraDataInterfaceSkeletalMesh_0}
   }
  }
 ]
}
