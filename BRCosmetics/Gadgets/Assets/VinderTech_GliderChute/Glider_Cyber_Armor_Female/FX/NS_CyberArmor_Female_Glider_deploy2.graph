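# Niagara deploy effect for the female Cyber Armor glider, combining electric-looking arcs, flickering light and flare sprites.
# LargeArcs002 bursts short-lived ribbons along randomly angled, random-height lines using a zero-radius cylinder ShapeLocation, with sine-driven alpha and tapered width.
# Light pairs sprite and light renderers with waveform flicker, while flare sprites grow and fade.
# All emitters tint from System.ColorPlusBoost; arc skeletal-mesh and CylinderLocation sampling are disabled.
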NiagaraSystem /BRCosmetics/Gadgets/Assets/VinderTech_GliderChute/Glider_Cyber_Armor_Female/FX/NS_CyberArmor_Female_Glider_deploy2
dependencies {
 usedDataInterfaces [&NiagaraDataInterfaceArrayFloat, &NiagaraDataInterfaceCurve]
 usedDynamicInputs []
 usedModules [
  &self:SystemSpawnScript.SystemScriptSource.SystemScriptGraph.NiagaraNodeAssignment_0.SetVariables_#1
  &SystemState
  &Timeline
 ]
}
emitters {
 LargeArcs002 {
  emitterSpawnScript [/SetVariables(Emitter.Num Particles=8) {isSetParameters=true}]
  emitterUpdateScript [
   &EmitterState()
   &SpawnBurst_Instantaneous(Spawn Count=$Emitter.Num Particles, Spawn Time=0.01, Use Spawn Probability=false, UseLoopCountLimit=false)
   /SetVariables(
    Emitter.Random Bone:NiagaraInt32=&RandomRangeInt(Minimum=0, Maximum=3)
    Emitter.random Float:NiagaraFloat=&RandomRangeFloat(Minimum=0, Maximum=1)
    Emitter.random Height:NiagaraFloat=&RandomRangeFloat(Minimum=35, Maximum=75, Evaluation Type=::NewEnumerator1)
    Emitter.random bool:NiagaraBool=&SetBoolByFloatComparison(A=&RandomRangeFloat(Minimum=0, Maximum=1, Evaluation Type=::NewEnumerator1), B=0.75)
    Emitter.random float 2:NiagaraFloat=&RandomRangeFloat(Minimum=0.65, Maximum=1, Evaluation Type=::NewEnumerator1)
    Emitter.random life:NiagaraFloat=&RandomRangeFloat(Minimum=0, Maximum=1, Evaluation Type=::NewEnumerator1)
    Emitter.random sine:NiagaraFloat=&RandomRangeFloat(Minimum=0.1, Maximum=2, Evaluation Type=::NewEnumerator1)
   ) {isSetParameters=true}
   /SetVariables(Emitter.Random Angle:Vector3f=&MakeVector(
     X=&RandomRangeFloat(Minimum=-0.5, Maximum=0.5, Evaluation Type=::NewEnumerator1)
     Y=&RandomRangeFloat(Minimum=-0.5, Maximum=0.5, Evaluation Type=::NewEnumerator1)
     Z=1
    )) {isSetParameters=true}
  ]
  particleSpawnScript [
   /SetVariables(
    Particles.Get Random Bool:NiagaraBool=$Emitter.random bool
    Particles.Get Random Height:NiagaraFloat=$Emitter.random Height
    Particles.get random float 2:NiagaraFloat=$Emitter.random float 2
    Particles.get random life:NiagaraFloat=$Emitter.random life
   ) {isSetParameters=true}
   &InitializeRibbon(
    Color=&MultiplyLinearColor(
     Linear Color=$System.ColorPlusBoost
     Scale Factor=&MakeFloat4FromVectorAndFloat("Vector (RGB)"=&VectorFromFloat(Value=5), "Float (Alpha)"=1)
    )
    Lifetime=0.2
    Ribbon Width=10
    UsePositionOffset=false
    Write Material Random=true
   )
   /SetVariables(Particles.SpriteSize=<5 5>) {isSetParameters=true}
   &FN_SetSpawnGroup(
    Inclusivity=true
    Normalize Spawn Group=true
    Particles Per Group:NiagaraInt32=$Emitter.Num Particles
    Pattern=::NewEnumerator0
    Set Ribbon ID and Ribbon Link Order=true
   )
   /SetVariables(Particles.Get random float:NiagaraFloat=$Emitter.random Float, Particles.get random bone:NiagaraInt32=$Emitter.Random Bone) {isSetParameters=true}
   /SetVariables(Particles.invert float:NiagaraFloat=&SelectFloatFromArray(Direct Array Index=$Particles.get random bone, Float Selection Array=DataInterface() {propertyValues {FloatData [-1,1,-1,1]}})) {isSetParameters=true}
   /SetVariables(Particles.Random Sine:NiagaraFloat=$Emitter.random sine) {isSetParameters=true}
   /SetVariables(Particles.Get Random Angle:Vector3f=$Emitter.Random Angle) {isSetParameters=true}
  ]
  particleUpdateScript [
   &ParticleState()
   &SkeletalMeshLocation(
    Bone / Socket Index=$Particles.get random bone
    Bone Sampling Mode=::NewEnumerator4
    MaskSpawnGroup=false
    Orientation Sampling=::NewEnumerator1
    Skeletal Mesh=$User.SkeletalMesh
    Use Sampled Mesh Attribute as Ribbon ID=false
    Write To Particle Position=false
   ) {enabled=false}
   &ScaleColor(
    Scale Alpha=&MakeCustomFloatFromBool(
     Bool=$Particles.Initial.Get Random Bool
     True Float=&Sine(Normalized Angle=$Emitter.Age, Period=$Particles.Initial.Get random float, Scale=1, Bias=0)
     False Float=&Sine(Normalized Angle=$Emitter.Age, Period=&RandomRangeFloat(Minimum=0.1, Maximum=1), Scale=1, Bias=0)
    )
    Scale RGB=&VectorFromFloat(Value=&FloatFromCurve(FloatCurve=DataInterface() {propertyValues {Curve {keys [<at 0 100>,<at 0.5 1>,<at 1 1>]}}}))
    ScaleA=true
    ScaleRGB=true
   )
   &ScaleRibbonWidth(Ribbon Width Scale=&Multiply_Float(
     A=&FloatFromCurve(FloatCurve=DataInterface() {propertyValues {Curve {keys [
          {interpMode=RCIM_Cubic, tangentMode=RCTM_User, time=0, value=0.1}
          {interpMode=RCIM_Cubic, tangentMode=RCTM_User, time=0.25, value=1}
          {arriveTangentWeight=0.1491, interpMode=RCIM_Cubic, leaveTangentWeight=0.2236, tangentMode=RCTM_User, time=0.75, value=1}
          {arriveTangentWeight=0.1, interpMode=RCIM_Cubic, tangentMode=RCTM_User, time=1, value=0.1}
         ]}}})
     B=&FloatFromCurve(
      FloatCurve=DataInterface() {propertyValues {Curve {keys [
          {arriveTangent=6.6108, interpMode=RCIM_Cubic, leaveTangent=6.6108, tangentMode=RCTM_User, time=0.05, value=0}
          {arriveTangentWeight=0.48, interpMode=RCIM_Cubic, leaveTangentWeight=0.48, time=0.5, value=1}
          {arriveTangent=-6.3941, interpMode=RCIM_Cubic, leaveTangent=-6.3941, tangentMode=RCTM_User, time=0.95, value=0}
         ]}}}
      CurveIndex=$Particles.Initial.Normalized Spawn Group
     )
    )) {module=ScaleRibbonWidth001}
   &CylinderLocation(
    Cylinder Axis:Vector3f=&ReturnMeshOrientationAxis(Mesh Orientation Axis=::ZAxis, Mesh Orientation Quat=$Particles.SkeletalMeshLocation.SampledRotation)
    Cylinder Distribution:ENiagara_CylinderMode=::NewEnumerator2
    Cylinder Height:NiagaraFloat=&Multiply_Float(
     A=$Particles.get random float 2
     B=&SelectFloatFromArray(Direct Array Index=$Particles.get random bone, Float Selection Array=DataInterface() {propertyValues {FloatData [75,75,45,45]}})
    )
    Cylinder Height Midpoint=0
    Cylinder Origin:NiagaraPosition=&ConvertVectorToPosition(
     Input Position=Linked() {linkedVariable {name=Particles.SkeletalMeshLocation.SampledPosition, type=NiagaraPosition}}
     Vector Is In Absolute Large World Space:NiagaraBool=false
    )
    Cylinder Radius=&FloatFromCurve(
     FloatCurve=DataInterface() {propertyValues {Curve {keys [
         {
          arriveTangent=5.1386
          arriveTangentWeight=1.1087
          interpMode=RCIM_Cubic
          leaveTangent=5.1386
          leaveTangentWeight=1.1087
          tangentMode=RCTM_User
          tangentWeightMode=RCTWM_WeightedBoth
          time=0
          value=0
         }
         {
          arriveTangentWeight=0.1963
          interpMode=RCIM_Cubic
          leaveTangentWeight=0.1963
          tangentMode=RCTM_User
          tangentWeightMode=RCTWM_WeightedBoth
          time=1
          value=1
         }
        ]}}}
     CurveIndex=$Particles.Normalized Spawn Group
     Scale Curve=&SelectFloatFromArray(Direct Array Index=$Particles.get random bone, Float Selection Array=DataInterface() {propertyValues {FloatData [10,10,7,7]}})
    )
    MaskSpawnGroup=false
    Non Uniform Scale:Vector3f=&Multiply_VectorByFloat(Vector=<1 1 1>, Float=1)
    Rotation Mode:ENiagara_RotationMode=::NewEnumerator1
    U Position:NiagaraFloat=&Multiply_Float(A=$Particles.Normalized Spawn Group, B=&RandomRangeFloat(Minimum=0, Maximum=3))
    UseOwnerScale=true
    V Position:NiagaraFloat=$Particles.Normalized Spawn Group
   ) {enabled=false}
   &ShapeLocation(
    Cylinder Distribution=::NewEnumerator2
    Cylinder Height=&FloatFromCurve(
     FloatCurve=DataInterface() {propertyValues {Curve {keys [{tangentMode=RCTM_User, time=0, value=0}, {tangentMode=RCTM_User, time=1, value=1}]}}}
     Scale Curve=$Particles.Get Random Height
    )
    Cylinder Height Midpoint=0
    Cylinder Radius=0
    MaskSpawnGroup=false
    Offset=<0 0 -5>
    Offset Mode=::NewEnumerator0
    Rotation Axis=$Particles.Get Random Angle
    Rotation Mode=::NewEnumerator1
    Shape Primitive=::NewEnumerator1
    U Position=&RandomRangeFloat(Minimum=0, Maximum=1)
    UseOwnerScale=true
    V Position=$Particles.Initial.Normalized Spawn Group
   )
  ]
  rendererClasses [&NiagaraRibbonRendererProperties]
  renderers [
   {rendererClass=&NiagaraRibbonRendererProperties}
   {rendererClass=&NiagaraRibbonRendererProperties, rendererIndex=1}
  ]
 }
 Light {
  emitterUpdateScript [
   &EmitterState()
   &SpawnRate(Use Spawn Probability=false)
   &SpawnBurst_Instantaneous(Age:NiagaraFloat=$Emitter.Age, Spawn Time=0.01, Use Spawn Probability=false, UseLoopCountLimit=false)
  ]
  particleSpawnScript [&InitializeParticle(
    Color=&HueShiftLinearColor(
     Hue=0
     Saturation=0.8
     Value=1
     Linear Color=&MultiplyLinearColor(
      Linear Color=$System.ColorPlusBoost
      Scale Factor=&MakeFloat4FromVectorAndFloat("Vector (RGB)"=&VectorFromFloat(Value=0.5), "Float (Alpha)"=0.25)
     )
    )
    Lifetime=0.5
    Position Offset=<0 0 15>
    UsePositionOffset=true
    Write Material Random=true
   )]
  particleUpdateScript [
   &ParticleState()
   &ScaleColor(
    Scale Alpha=&Waveform(
     Waveform Count=::NewEnumerator2
     Global Phase=&Multiply_Float(A=$Emitter.Age, B=3)
     Global Amplitude Scale=100
     "Waveform [1]"=::NewEnumerator1
     "[1] Frequency"=1
     "[1] Amplitude Min/Max"=<0 1>
     "[2] Average"=false
     "[2] Input Position"=0
     "[2] Phase"=$Emitter.Age
     "[2] Phase Scale"=1
     "[2] Compound Frequency"=2
     "[2] Frequency"=2
     "[2] Amplitude Min/Max"=<0 1>
     "[2] Scale"=1
     "[2] Bias"=0
     "[2] Exponent"=1
     "[2] Pulse Duration"=0.1
     "[2] Clip Triangle Min/Max"=<0 0>
     "[2] Random Seed"=0
     "[3] Average"=false
     "[3] Input Position"=0
     "[3] Phase"=$Emitter.Age
     "[3] Phase Scale"=1
     "[3] Frequency"=4
     "[3] Compound Frequency"=2
     "[3] Amplitude Min/Max"=<0 1>
     "[3] Scale"=1
     "[3] Bias"=0
     "[3] Exponent"=1
     "[3] Pulse Duration"=0.1
     "[3] Clip Triangle Min/Max"=<0 0>
     "[3] Random Seed"=0
    )
    ScaleA=true
    ScaleRGB=true
   )
  ]
  rendererClasses [&NiagaraSpriteRendererProperties, &NiagaraLightRendererProperties]
  renderers [
   {rendererClass=&NiagaraSpriteRendererProperties}
   {rendererClass=&NiagaraLightRendererProperties, rendererIndex=1}
  ]
 }
 flare {
  emitterUpdateScript [
   &EmitterState(Life Cycle Mode=::NewEnumerator1, Loop Behavior=::NewEnumerator1, Loop Duration=1, UseLoopDelay=false)
   &SpawnBurst_Instantaneous(Spawn Time=0.01, Use Spawn Probability=false, UseLoopCountLimit=false)
  ]
  particleSpawnScript [
   &InitializeParticle(
    Color=&MultiplyLinearColor(
     Linear Color=$System.ColorPlusBoost
     Scale Factor=&MakeFloat4FromVectorAndFloat("Vector (RGB)"=&VectorFromFloat(Value=5), "Float (Alpha)"=10)
    )
    Lifetime Max=0.35
    Lifetime Min=0.4
    Lifetime Mode=::NewEnumerator1
    Sprite Size Mode=::NewEnumerator3
    Uniform Sprite Size=45
    UsePositionOffset=false
    Write Material Random=true
   )
   &CameraOffset(Camera Offset Amount=2.5)
  ]
  particleUpdateScript [
   &ParticleState()
   &ScaleColor(
    Scale Alpha=&FloatFromCurve(FloatCurve=DataInterface() {propertyValues {Curve {keys [
         {
          arriveTangent=-2.4268
          arriveTangentWeight=0.9047
          interpMode=RCIM_Cubic
          leaveTangent=-2.4268
          leaveTangentWeight=0.9047
          tangentMode=RCTM_User
          tangentWeightMode=RCTWM_WeightedBoth
          time=-0.0046
          value=1.0113
         }
         {arriveTangentWeight=0.3352, interpMode=RCIM_Cubic, tangentMode=RCTM_User, tangentWeightMode=RCTWM_WeightedBoth, time=1.0011, value=0.002}
        ]}}})
    ScaleA=true
    ScaleRGB=true
   )
   &ScaleSpriteSize(Uniform Curve Sprite Scale=DataInterface() {propertyValues {Curve {keys [{arriveTangent=1.135, interpMode=RCIM_Cubic, leaveTangent=1.135, tangentMode=RCTM_User, time=0, value=0}, <at 1 1 RCIM_Cubic>]}}})
  ]
  rendererClasses [&NiagaraSpriteRendererProperties]
  renderers [
   {rendererClass=&NiagaraSpriteRendererProperties}
   {rendererClass=&NiagaraSpriteRendererProperties, rendererIndex=1}
  ]
 }
}
summary {
 systemName=NS_CyberArmor_Female_Glider_deploy2
 emitters [
  {emitterName=LargeArcs002, rendererClasses [&NiagaraRibbonRendererProperties]}
  {emitterName=Light, rendererClasses [&NiagaraSpriteRendererProperties, &NiagaraLightRendererProperties]}
  {emitterName=flare, rendererClasses [&NiagaraSpriteRendererProperties]}
 ]
 userVariables [
  User.Color {type=LinearColor, description="", defaultValue {struct=LinearColor, value=<rgba 0.3438 0 0.0108 1>}}
  User.SkeletalMesh {
   type=NiagaraDataInterfaceSkeletalMesh
   defaultValue {
    struct=NiagaraExt_VariableValue_DataInterface
    value {dataInterface=&self:NiagaraDataInterfaceSkeletalMesh_0}
   }
  }
 ]
}
